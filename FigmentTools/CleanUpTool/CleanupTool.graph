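# Editor utility widget "Cleanup Tool" that gathers settings for cleaning up assets in a chosen directory.
# PreConstruct binds single-property views to its Directory and MapToSearch variables; an empty map means references are searched in any map.
# Execute constructs a CleanUpToolModel and calls CleanUpAssets with path, map, the HLOD-deletion checkbox and a marked-for-add filter.
# That filter comes from negating the combo index, so "Marked for Add" enables it.
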EditorUtilityWidgetBlueprint /FigmentTools/CleanUpTool/CleanupTool
graphs [EventGraph {nodes [
   "On Clicked (EditorUtilityButton_Execute)"() {is=title K2Node_ComponentBoundEvent}
   > n1 = Construct Object from Class(Class=/Script/FigmentTools.CleanUpToolModel) -> (ReturnValue:Clean Up Tool Model Object Reference)
   n2 = GetSelectedIndex!(self:"Combo Box (String) Object Reference"=Get ComboBox_ExecutionMethod() -> (ComboBox_ExecutionMethod:Editor Utility Combo Box String Object Reference)) -> (ReturnValue:int=0)
   n3 = Get CheckedState!(self:Check Box Object Reference=Get EditorUtilityCheckBox_HLODS() -> (EditorUtilityCheckBox_HLODS:Editor Utility Check Box Object Reference)) -> (CheckedState:ECheckBoxState Enum=Unchecked)
   > CleanUpAssets!(
    execute
    self:Clean Up Tool Model Object Reference=@n1
    InDirectoryPathToSearch:String=Break Directory Path!(DirectoryPath:"Directory Path Structure (by ref)"=Get Directory() -> (Directory:Directory Path Structure)) -> (Path:String) {is=title K2Node_BreakStruct}
    bFilterMarkForAdd:bool=NOT Boolean(A="To Boolean (Integer)"!(InInt:int=@n2) -> (ReturnValue:bool))
    bDeleteGeneratedHLODs:bool="To Boolean (Integer)"(InInt="To Integer (Byte)"!(InByte:Byte=@n3) -> (ReturnValue:int=0))
    InSpecificMap:Object Reference=Get MapToSearch() -> (MapToSearch:Object Reference)
   ) -> (then)
   Event PreConstruct()
   n4 = Self-Reference()
   > SetObject(self=Get SinglePropertyView_Directory() -> (SinglePropertyView_Directory:Single Property View Object Reference), NewObject=@n4)
   > SetObject(self=Get SinglePropertyView_Map() -> (SinglePropertyView_Map:Single Property View Object Reference), NewObject=@n4)
  ]}]
variables [
 Directory {type=DirectoryPath, default {path=""}}
 MapToSearch {type=Object, default=null}
]
widgets {
 info {rootWidgetClass=&VerticalBox, widgetCount=13}
 widgets [
  VerticalBox_0 {widgetClassPath=&VerticalBox, widget=VerticalBox_0}
  TextBlock_82 {
   widgetClassPath=&TextBlock
   parent=VerticalBox_0
   widget=TextBlock_82
   properties {
    justification=Center
    text=Cleanup Tool
    font {
     bForceMonospaced=false
     bMaterialIsStencil=false
     fontMaterial=null
     fontObject=&Roboto
     letterSpacing=0
     monospacedWidth=1
     size=16.5
     skewAmount=0
     typefaceFontName=Bold
     outlineSettings {
      bApplyOutlineToDropShadows=false
      bMiteredCorners=false
      bSeparateFillAlpha=false
      outlineMaterial=null
      outlineSize=0
      outlineColor=<rgba 0 0 0 1>
     }
    }
   }
   slot {class=VerticalBoxSlot, properties {verticalAlignment=VAlign_Top}}
  }
  SinglePropertyView_Directory {
   widgetClassPath=&SinglePropertyView
   parent=VerticalBox_0
   widget=SinglePropertyView_Directory
   properties {propertyName=Directory}
   slot {class=VerticalBoxSlot, properties {verticalAlignment=VAlign_Top}}
  }
  VerticalBox_1 {
   widgetClassPath=&VerticalBox
   parent=VerticalBox_0
   widget=VerticalBox_1
   slot {class=VerticalBoxSlot, properties {verticalAlignment=VAlign_Center, padding=<ltrb 5 5 5 5>}}
  }
  Text_MapToSearch {
   widgetClassPath=&TextBlock
   parent=VerticalBox_1
   widget=Text_MapToSearch
   properties {
    text=Select a specific map or leave it empty to look for references in any map.
    colorAndOpacity {colorUseRule=UseColor_Specified, specifiedColor=<rgba 1 0.8781 0.4139 1>}
    font {
     bForceMonospaced=false
     bMaterialIsStencil=false
     fontMaterial=null
     fontObject=&Roboto
     letterSpacing=0
     monospacedWidth=1
     size=7.5
     skewAmount=0
     typefaceFontName=Bold
     outlineSettings {
      bApplyOutlineToDropShadows=false
      bMiteredCorners=false
      bSeparateFillAlpha=false
      outlineMaterial=null
      outlineSize=0
      outlineColor=<rgba 0 0 0 1>
     }
    }
   }
   slot {class=VerticalBoxSlot, properties {horizontalAlignment=HAlign_Center}}
  }
  SinglePropertyView_Map {
   widgetClassPath=&SinglePropertyView
   parent=VerticalBox_1
   widget=SinglePropertyView_Map
   properties {bAutoLoadAsset=false, propertyName=MapToSearch}
   slot {class=VerticalBoxSlot}
  }
  HorizontalBox_0 {
   widgetClassPath=&HorizontalBox
   parent=VerticalBox_0
   widget=HorizontalBox_0
   slot {class=VerticalBoxSlot, properties {verticalAlignment=VAlign_Top, padding=<ltrb 5 5 5 5>}}
  }
  EditorUtilityCheckBox_HLODS {
   widgetClassPath=&EditorUtilityCheckBox
   bIsVariable=true
   parent=HorizontalBox_0
   widget=EditorUtilityCheckBox_HLODS
   slot {class=HorizontalBoxSlot}
  }
  TextBlock_0 {
   widgetClassPath=&TextBlock
   parent=HorizontalBox_0
   widget=TextBlock_0
   properties {
    text=Delete generated HLODS
    font {
     bForceMonospaced=false
     bMaterialIsStencil=false
     fontMaterial=null
     fontObject=&Roboto
     letterSpacing=0
     monospacedWidth=1
     size=6.75
     skewAmount=0
     typefaceFontName=Bold
     outlineSettings {
      bApplyOutlineToDropShadows=false
      bMiteredCorners=false
      bSeparateFillAlpha=false
      outlineMaterial=null
      outlineSize=0
      outlineColor=<rgba 0 0 0 1>
     }
    }
   }
   slot {class=HorizontalBoxSlot, properties {horizontalAlignment=HAlign_Center, verticalAlignment=VAlign_Center}}
  }
  HorizontalBox_1 {
   widgetClassPath=&HorizontalBox
   parent=VerticalBox_0
   widget=HorizontalBox_1
   slot {class=VerticalBoxSlot, properties {verticalAlignment=VAlign_Bottom, padding=<ltrb 5 5 5 5>}}
  }
  EditorUtilityButton_Execute {
   widgetClassPath=&EditorUtilityButton
   parent=HorizontalBox_1
   widget=EditorUtilityButton_Execute
   slot {class=HorizontalBoxSlot, properties {padding=<ltrb 5 5 5 5>, size {sizeRule=Fill, value=1}}}
  }
  TextBlock_31 {
   widgetClassPath=&TextBlock
   parent=EditorUtilityButton_Execute
   widget=TextBlock_31
   properties {
    text=Execute
    font {
     bForceMonospaced=false
     bMaterialIsStencil=false
     fontMaterial=null
     fontObject=&Roboto
     letterSpacing=0
     monospacedWidth=1
     size=12
     skewAmount=0
     typefaceFontName=Bold
     outlineSettings {
      bApplyOutlineToDropShadows=false
      bMiteredCorners=false
      bSeparateFillAlpha=false
      outlineMaterial=null
      outlineSize=0
      outlineColor=<rgba 0 0 0 1>
     }
    }
   }
   slot {class=ButtonSlot}
  }
  ComboBox_ExecutionMethod {
   widgetClassPath=/Script/Blutility.EditorUtilityComboBoxString
   bIsVariable=true
   parent=HorizontalBox_1
   widget=ComboBox_ExecutionMethod
   properties {
    selectedOption=Marked for Add
    defaultOptions [Marked for Add, All assets]
    font {
     bForceMonospaced=false
     bMaterialIsStencil=false
     fontMaterial=null
     fontObject=&Roboto
     letterSpacing=0
     monospacedWidth=1
     size=10.5
     skewAmount=0
     typefaceFontName=Bold
     outlineSettings {
      bApplyOutlineToDropShadows=false
      bMiteredCorners=false
      bSeparateFillAlpha=false
      outlineMaterial=null
      outlineSize=0
      outlineColor=<rgba 0 0 0 1>
     }
    }
    foregroundColor {colorUseRule=UseColor_Specified, specifiedColor=<rgba 0.5271 0.5271 0.5271 1>}
   }
   slot {class=HorizontalBoxSlot, properties {verticalAlignment=VAlign_Center, padding=<ltrb 5 5 5 5>}}
  }
 ]
}
